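; Armored window vehicle mod component that toughens a vehicle's windows and changes their look.
; BeginPlay caches the owning Athena vehicle and swaps one skeletal mesh material slot to a metal debris material, choosing the slot from Valet SUV or sports car gameplay tags.
; Replicated mod config supplies the armored window health via a tagged script data key.
; On enable, the authority overrides each named window part's max health with that value.
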(event EventBeginPlay
  ;[1 1 1] EVENT -> BeginPlay - Set Window Material
  (bind _asfort_athena_vehicle (Utilities|Casting|CastToFortAthenaVehicle (Components|GetOwner)))
  (bind _output_get (Variables|Default|SetOwnerVehicle _asfort_athena_vehicle))
  (if (GameplayTags|HasMatchingGameplayTag _output_get "(TagName=\"Vehicle.Valet.BasicSUV\")")
    (Rendering|Material|SetMaterial (Variables|/Script/FortniteGame.FortAthenaVehicle|GetSkeletalMesh _output_get) 2 "/Game/Effects/Fort_Effects/Mesh/Debris/M_Metal_Frag_Chunk_FX.M_Metal_Frag_Chunk_FX")
    (else
      (Rendering|Material|SetMaterial (Variables|/Script/FortniteGame.FortAthenaVehicle|GetSkeletalMesh _output_get) (Math|Integer|SelectInt 0 1 (GameplayTags|HasMatchingGameplayTag _output_get "(TagName=\"Athena.Vehicle.Valet.SportsCar\")")) "/Game/Effects/Fort_Effects/Mesh/Debris/M_Metal_Frag_Chunk_FX.M_Metal_Frag_Chunk_FX"))))

(event VehicleMods|EventOnModConfigReplicated
  (params (ModConfigData :type Fort%20Vehicle%20Mod%20Config%20Object%20Reference))
  ;[1 1 1] EVENT -> OnConfigReplicated: Get config data for Window Health
  (Utilities|IsValid ModConfigData
    (:Is%20Valid
      (bind _outscriptdata (Class|FortVehicleModConfig|GetScriptDataforKey ModConfigData "(TagName=\"Vehicle.Mod.Window.Armored\")")
        (:Valid
          (Variables|Default|SetArmoredWindowHealth (Utilities|Struct|BreakScalableFloat (.FloatData (Utilities|Struct|BreakFortVehicleModScriptData _outscriptdata)))))))))

(event VehicleMods|EventOnModEnabled
  ;[1 1 1] EVENT -> OnModEnabled: Set Armored Window Health
  (bind _ownervehicle (Variables|Default|GetOwnerVehicle)
    (:then
      (Networking|SwitchHasAuthority
        (:Authority
          (for _array_element (Variables|Default|GetWindowNames)
            (Class|FortAthenaVehicle|OverrideDamageablePartMaxHealth _ownervehicle _array_element (Variables|Default|GetArmoredWindowHealth) true)))))
    (:else)))
